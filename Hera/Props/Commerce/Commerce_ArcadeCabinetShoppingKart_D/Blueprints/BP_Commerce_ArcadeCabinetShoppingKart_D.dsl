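; Arcade cabinet shopping cart prop Blueprint whose construction script picks a music loop to match the cabinet's material.
; It reads the static mesh material at slot index 1 and looks it up in the AudioMap.
; When music is not muted, the material is valid and a map entry is found, it sets that sound as the Arcade Music prop audio component's loop.
; BeginPlay, overlap and Tick events are empty.
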(event EventBeginPlay)

(event Collision|EventActorBeginOverlap
  (params (OtherActor :type Actor%20Object%20Reference)))

(event EventTick
  (params (DeltaSeconds :type Float%20%28single-precision%29)))

(fn ConstructionScript
  (bind _staticmeshcomponent (Variables|Debug|GetStaticMeshComponent))
  (bind _returnvalue (Rendering|Material|GetMaterial _staticmeshcomponent 1))
  (bind _audiomap (Variables|Default|GetAudioMap))
  (bind (_value _returnvalue_2) (Utilities|Map|Find _audiomap _returnvalue))
  (|CallParentFunction)
  (if (not (Variables|Default|GetMute%20Music))
    (Utilities|IsValid _returnvalue
      (:Is%20Valid
        (if _returnvalue_2
          (Variables|/Game/Sounds/Fort_Audio_Blueprints/PropAmbientAudioComponent/FortPropAudioComponent.FortPropAudioComponent_C|SetSoundLoop _value (Variables|Default|GetArcade_Music)))))))
